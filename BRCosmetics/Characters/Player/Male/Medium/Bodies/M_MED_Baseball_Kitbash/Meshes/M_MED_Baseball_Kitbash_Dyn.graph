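# Physics asset for the medium male Baseball Kitbash body, covering the left arm and its dynamic shoulder pad bone.
# Capsule bodies sit on the pad, upper arm, upper-arm twist and lower arm, with the larger upper-arm capsule matching the arm's volume.
# The pad is constrained to the twist bone like a hinge, with twist and one swing locked and the other swing given a narrow limit.
# The twist and lower-arm bodies join the upper arm through angularly free constraints.
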PhysicsAsset /BRCosmetics/Characters/Player/Male/Medium/Bodies/M_MED_Baseball_Kitbash/Meshes/M_MED_Baseball_Kitbash_Dyn
bodies {
 dyn_shoulderpad_l [{shapeType=Capsule, length=14.75, radius=1.75, center=<7.2843 -0.0002 0>, rotation=<rot 90 0 0>}]
 lowerarm_l [{shapeType=Capsule, length=2, radius=2, rotation=<rot 90 0 0>}]
 upperarm_l [{shapeType=Capsule, length=7, radius=5.75, center=<11.7689 -0.0003 0>, rotation=<rot 90 0 0>}]
 upperarm_twist_01_l [{shapeType=Capsule, length=2, radius=2, rotation=<rot 90 0 0>}]
}
constraints [
 {
  bone1Name=dyn_shoulderpad_l
  bone2Name=upperarm_twist_01_l
  swing1LimitDegrees=5
  swing1Motion=Locked
  swing2LimitDegrees=25
  twistLimitDegrees=2
  twistMotion=Locked
 }
 {
  bone1Name=upperarm_twist_01_l
  bone2Name=upperarm_l
  swing1LimitDegrees=45
  swing1Motion=Free
  swing2LimitDegrees=45
  swing2Motion=Free
  twistLimitDegrees=45
  twistMotion=Free
 }
 {
  bone1Name=lowerarm_l
  bone2Name=upperarm_l
  swing1LimitDegrees=45
  swing1Motion=Free
  swing2LimitDegrees=45
  swing2Motion=Free
  twistLimitDegrees=45
  twistMotion=Free
 }
]
