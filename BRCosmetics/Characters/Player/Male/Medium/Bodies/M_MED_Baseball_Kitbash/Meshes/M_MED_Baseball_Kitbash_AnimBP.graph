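# Post-copy animation Blueprint for the Baseball Kitbash male body mesh, adding secondary motion to the character cosmetic.
# It copies the pose from the attached parent mesh, then constrains the right knee pad between thigh and calf bones.
# A RigidBody node simulates the Dyn physics asset in spine_01 space with damped component motion and overridden gravity.
# Each update maps pawn 2D speed to the simulation alpha, reducing physics influence at high speed.
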AnimBlueprint /BRCosmetics/Characters/Player/Male/Medium/Bodies/M_MED_Baseball_Kitbash/Meshes/M_MED_Baseball_Kitbash_AnimBP
target_skeleton=&/M_MED_Baseball_Kitbash_Skeleton
graphs [
 AnimGraph {nodes [
   n1 = "Constraint - knee_pad_r (PP)"!(
    ComponentPose=Local To Component(LocalPose=Copy Pose From Mesh() {bUseAttachedParent=true})
    Alpha=1
    ConstraintWeights_0=0.6
    ConstraintWeights_1:float=0.4
   ) -> (Pose) {
    alphaBoolBlend {blendOption=Linear}
    boneToModify {boneName=knee_pad_r}
    constraintSetup [{targetBone {boneName=thigh_r}}, {targetBone {boneName=calf_r}}]
    constraintWeights [1,1]
   }
   n2 = RigidBody(ComponentPose=@n1, Alpha=Get speed alpha() -> (speed alpha:double)) {
    bClampLinearTranslationLimitToRefPose=true
    bOverrideWorldGravity=true
    overridePhysicsAsset=&/M_MED_Baseball_Kitbash_Dyn
    simulationSpace=BaseBoneSpace
    alphaBoolBlend {blendOption=Linear}
    baseBoneRef {boneName=spine_01}
    componentAppliedLinearAccClamp=<3000 3000 3000>
    componentLinearAccScale=<0.4 0.3 0.3>
    componentLinearVelScale=<0.4 0.3 0.3>
    overrideWorldGravity {z=-980}
   }
   Output Pose(Result=Component To Local(ComponentPose=@n2)) {layerGroup=None}
  ]}
 EventGraph {nodes [
   Event BlueprintUpdateAnimation()
   > Set speed alpha(speed alpha:double=MapRangeClamped(Value=Get PawnSpeed2D(), InRangeA=450, InRangeB=475, OutRangeA=1, OutRangeB=0.3)) -> (Output_Get:double)
   TryGetPawnOwner()
  ]}
]
variables [speed alpha {type=number}]
